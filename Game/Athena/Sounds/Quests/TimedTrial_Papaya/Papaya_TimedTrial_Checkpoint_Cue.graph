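# Sound cue for a checkpoint in a Papaya timed trial quest.
# A single wave player feeds the checkpoint wave straight to the output, with no randomization or modulation.
# The cue boosts volume, routes through the quests sound class and borrows the gadget-activate attenuation settings for spatial falloff.
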SoundCue /Game/Athena/Sounds/Quests/TimedTrial_Papaya/Papaya_TimedTrial_Checkpoint_Cue
root=WavePlayer_2
nodes [{id=WavePlayer_2, properties {soundWaveAssetPtr=&/Papaya_TimedTrial_Checkpoint_01}}]
properties {
 attenuationSettings=&Fort_GadgetActivate_Att
 maxDistance=4000
 soundClassObject=&Fort_Quests
 volumeMultiplier=2.24
}
