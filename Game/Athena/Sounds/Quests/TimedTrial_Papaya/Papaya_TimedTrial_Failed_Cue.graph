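# Sound cue that plays the failure sound for the Papaya timed trial quest.
# A single wave player feeds one failure sound wave straight to the output with boosted volume.
# It routes through the Fort quests sound class and overrides concurrency to cap simultaneous instances.
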SoundCue /Game/Athena/Sounds/Quests/TimedTrial_Papaya/Papaya_TimedTrial_Failed_Cue
root=WavePlayer_0
nodes [{id=WavePlayer_0, properties {soundWaveAssetPtr=&/Papaya_TimedTrial_Fail_02}}]
properties {
 bOverrideConcurrency=true
 maxDistance=2097152
 soundClassObject=&Fort_Quests
 volumeMultiplier=1.792
 concurrencyOverrides {maxCount=2}
}
