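# Sound cue for the score sound of the Papaya timed trial quest in Athena.
# A single wave player plays the Papaya_TimedTrial_Score_01 wave with boosted volume, routed through the Fort_Quests sound class.
# Concurrency is overridden to cap simultaneous instances at a small count.
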SoundCue /Game/Athena/Sounds/Quests/TimedTrial_Papaya/Papaya_TimedTrial_Score_cue.Papaya_TimedTrial_Score_Cue
root=WavePlayer_0
nodes [{id=WavePlayer_0, properties {soundWaveAssetPtr=&/Papaya_TimedTrial_Score_01}}]
properties {
 bOverrideConcurrency=true
 maxDistance=2097152
 soundClassObject=&Fort_Quests
 volumeMultiplier=2.24
 concurrencyOverrides {maxCount=2}
}
