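# Sound cue for the start of a Papaya timed trial quest.
# It plays a single wave, Papaya_TimedTrial_Start_01, through the quests sound class with boosted volume and a borrowed gadget-placement attenuation.
# A concurrency override allows only one instance per owner, stopping the oldest when a new one plays.
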SoundCue /Game/Athena/Sounds/Quests/TimedTrial_Papaya/Papaya_TimedTrial_Start_Cue
root=WavePlayer_2
nodes [{id=WavePlayer_2, properties {soundWaveAssetPtr=&/Papaya_TimedTrial_Start_01}}]
properties {
 attenuationSettings=/Game/Sounds/Attenuation/Fort_GadgetPlace_Att
 bOverrideConcurrency=true
 maxDistance=2000
 soundClassObject=&Fort_Quests
 volumeMultiplier=2.24
 concurrencyOverrides {bLimitToOwner=true, maxCount=1, resolutionRule=StopOldest}
}
