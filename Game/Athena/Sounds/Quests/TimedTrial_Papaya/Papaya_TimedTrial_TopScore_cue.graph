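# Sound cue for a top score in the Papaya timed trial, playing a celebratory firework and crowd cheer.
# After a brief initial delay it mixes a slightly lowered firework launch with a half-second delayed burst of explosion cheer and huge firework explosion.
# Each burst layer passes through its own modulator for variation.
# It uses the Quests sound class and borrows the Crucible scoreboard attenuation.
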SoundCue /Game/Athena/Sounds/Quests/TimedTrial_Papaya/Papaya_TimedTrial_TopScore_cue.Papaya_TimedTrial_TopScore_Cue
root=Delay_1
nodes [
 {id=Delay_0, children [Mixer_0], properties {delayMax=0.5, delayMin=0.5}}
 {id=Delay_1, children [Mixer_1], properties {delayMax=0.1, delayMin=0.1}}
 {id=Mixer_0, children [Modulator_1,Modulator_0], properties {inputVolume [1,1]}}
 {id=Mixer_1, children [WavePlayer_3,Delay_0], properties {inputVolume [0.67,1]}}
 {id=Modulator_0, children [WavePlayer_2]}
 {id=Modulator_1, children [WavePlayer_0]}
 {id=WavePlayer_0, properties {soundWaveAssetPtr=/Game/Athena/Sounds/LTMs/Stage/Big_Explosion_Cheer}}
 {id=WavePlayer_2, properties {soundWaveAssetPtr=/Game/Sounds/Fort_GamePlay_Sounds/Explosives/Firework_Huge_Explosion_01}}
 {id=WavePlayer_3, properties {soundWaveAssetPtr=/Game/Sounds/Fort_GamePlay_Sounds/Explosives/Firework_Huge_Launch_01}}
]
properties {
 attenuationSettings=/Game/Athena/Sounds/LTMs/Crucible/Crucible_Scoreboard_Att
 maxDistance=6000
 soundClassObject=&Fort_Quests
 volumeMultiplier=0.896
}
